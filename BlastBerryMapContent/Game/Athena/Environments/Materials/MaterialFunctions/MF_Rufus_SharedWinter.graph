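# Shared winter material function that outputs time-of-day driven SnowColor and IceColor for environment materials.
# Morning, day, evening and night color parameters feed Color_Time_of_Day separately for snow and ice.
# Ice blends from a base IceColor toward that result by IceTintValue.
# Low quality darkens snow with a multiplier, adding a tint blend on the mobile shading path; other quality levels use the time-of-day color directly.
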MaterialFunction /BlastBerryMapContent/Game/Athena/Environments/Materials/MaterialFunctions/MF_Rufus_SharedWinter
graph {nodes [
  Constant3Vector() {desc=Base Evening Color, constant=<rgba 0.7561 0.8732 0.974 0>}
  Constant3Vector() {constant=<rgba 3 3 3 0>}
  Constant3Vector() {constant=<rgba 0.0812 0.1748 0.2708 1>}
  Constant3Vector() {desc=Base Night Color, constant=<rgba 0.7682 0.762 0.8073 0>}
  Constant3Vector() {desc=Base Snow Morning, constant=<rgba 0.9386 0.9284 0.9635 0>}
  Constant3Vector() {desc=Base Days, constant=<rgba 1 1 1 0>}
  Constant3Vector() {constant=<rgba 0.1199 0.2582 0.4 1>}
  Constant3Vector() {constant=<rgba 0.1199 0.2582 0.4 1>}
  Constant3Vector() {constant=<rgba 0.0889 0.1977 0.4 1>}
  Constant3Vector() {constant=<rgba 0.1199 0.2582 0.4 1>}
  Constant3Vector() {constant=<rgba 0.0741 0.1648 0.3333 1>}
  n1 = MaterialFunctionCall(
   Color_Morning=VectorParameter() {group=Colors, parameterName=SnowMorning, sortPriority=5, defaultValue=<rgba 0.9316 1.0758 1.2 0>}.RGB
   Color_Day=VectorParameter() {group=Colors, parameterName=SnowDay, sortPriority=6, defaultValue=<rgba 1 1 1 0>}.RGB
   Color_Evening=VectorParameter() {group=Colors, parameterName=SnowEvening, sortPriority=7, defaultValue=<rgba 0.9316 1.0758 1.2 0>}.RGB
   Color_Night=VectorParameter() {group=Colors, parameterName=SnowNight, sortPriority=8, defaultValue=<rgba 1.1419 1.1327 1.2 0>}.RGB
  ) {materialFunction=&Color_Time_of_Day}
  n2 = ShadingPathSwitch(
   Default=@n1.Result
   Mobile=LinearInterpolate(
    A=@n1.Result
    B=VectorParameter() {group=Colors, parameterName=SnowTintColor, defaultValue=<rgba 0.0656 0.1412 0.2188 1>}.RGB
    Alpha=ScalarParameter() {defaultValue=0.25, group=Colors, parameterName=Snow_Tint_Value}
   )
  ) {desc=Migrated from FeatureLevelSwitch}
  n3 = ScalarParameter() {defaultValue=0.35, group=Colors, parameterName=Snow_Low_EndMulti}
  FunctionOutput(None=QualitySwitch(Default=@n1.Result, Low=Multiply(A=@n2, B=@n3))) {outputName=SnowColor}
  n4 = VectorParameter() {group=Colors, parameterName=IceMorning, sortPriority=1, defaultValue=<rgba 0.1199 0.2582 0.4 1>}
  n5 = VectorParameter() {group=Colors, parameterName=IceDay, sortPriority=2, defaultValue=<rgba 0.1199 0.2582 0.4 1>}
  n6 = VectorParameter() {group=Colors, parameterName=IceEvening, sortPriority=3, defaultValue=<rgba 0.0778 0.173 0.35 1>}
  n7 = VectorParameter() {group=Colors, parameterName=IceNight, sortPriority=4, defaultValue=<rgba 0.1098 0.2001 0.31 1>}
  FunctionOutput(None=LinearInterpolate(
    A=VectorParameter() {group=Colors, parameterName=IceColor, defaultValue=<rgba 0.0656 0.1412 0.2188 1>}.RGB
    B=MaterialFunctionCall(Color_Morning=@n4.RGB, Color_Day=@n5.RGB, Color_Evening=@n6.RGB, Color_Night=@n7.RGB) {materialFunction=&Color_Time_of_Day}.Result
    Alpha=ScalarParameter() {defaultValue=0.5, group=Colors, parameterName=IceTintValue}
   ) {constAlpha=0.25}) {outputName=IceColor}
 ]}
