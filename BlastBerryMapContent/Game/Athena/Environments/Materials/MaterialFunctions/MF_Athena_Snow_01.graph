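# Snow surface material function for the BlastBerry Athena environment, outputting material attributes plus base color, normal, height and view-angle values.
# A world-XY mask texture, optionally landscape-mapped on Forward, modulates shared winter snow color, roughness, specular and height.
# Higher quality levels shade color with a Fresnel term blended by view-to-sun angle; lower levels and mobile simplify it.
# A larger-scale macro normal is mostly flattened, and height fades toward neutral with distance.
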MaterialFunction /BlastBerryMapContent/Game/Athena/Environments/Materials/MaterialFunctions/MF_Athena_Snow_01
graph {nodes [
  Constant() {desc=Previous values before asteria spec match, r=1}
  Constant() {r=0.35}
  n1 = Constant3Vector() {constant=<rgba 0.5734 0.72 0.85 1>}
  n2 = CameraVectorWS()
  n3 = DotProduct(A=PixelNormalWS(), B=@n2)
  n4 = Power(Base=OneMinus(None=Clamp(None=@n3)), Exp=Constant() {r=8}) {constExponent=6}
  n5 = OneMinus(None=@n4)
  n6 = CollectionParameter() {
   collection=&FortniteMaterialParameters
   parameterName=SunlightDirectionalVector
  }
  n7 = Clamp(None=DotProduct(A=@n2, B=ComponentMask(None=@n6) {b=true, g=true, r=true}))
  n8 = Constant() {r=0.5}
  n9 = MaterialFunctionCall() {materialFunction=&/MF_Rufus_SharedWinter}
  n10 = VectorParameter() {desc=used to adjust texture values, group=Colors, parameterName=Base Snow Color Multiplier, defaultValue=<rgba 0.91 0.85 0.85 0>}
  n11 = Divide(A=ComponentMask(None=WorldPosition().XYZ) {g=true, r=true}, B=Constant() {r=1200})
  n12 = StaticSwitch(
   True=ShadingPathSwitch(Default=@n11, Deferred=@n11, Forward=LandscapeLayerCoords() {mappingScale=6}, Mobile=@n11)
   False=@n11
   Value=FunctionInput(Preview=StaticBool()) {bUsePreviewValueAsDefault=true, inputName=UseLandscapeCoordsOnSwitch, inputType=FunctionInput_StaticBool, sortPriority=15}
  )
  n13 = TextureSample(UVs=@n12) {
   samplerSource=SSM_Wrap_WorldGroupSettings
   samplerType=SAMPLERTYPE_LinearColor
   texture=/BlastBerryMapContent/Game/Athena/Environments/Landscape/Textures/T_Rufus_Snow_01_M
  }
  n14 = Multiply(A=@n9.SnowColor, B=Multiply(A=@n10.RGB, B=@n13.B))
  n15 = QualitySwitch(Default=@n14, Low=@n9.SnowColor, High=@n14, Medium=@n9.SnowColor, Epic=@n14)
  n16 = FunctionOutput(None=ShadingPathSwitch(Default=@n15, Deferred=@n15, Mobile=Multiply(A=@n9.SnowColor)) {desc=Migrated from FeatureLevelSwitch}) {outputName=BaseColor}
  n17 = Multiply(
   A=LinearInterpolate(
    A=LinearInterpolate(A=@n1, Alpha=@n4) {constA=1}
    B=LinearInterpolate(Alpha=@n5) {constA=0.3, constB=0.7}
    Alpha=QualitySwitch(Default=@n7, Low=@n8, High=@n7, Medium=@n8)
   )
   B=@n16
  )
  n18 = LinearInterpolate(
   A=ScalarParameter() {defaultValue=0.2, group=Snow, parameterName=Snow_Roughness Low}
   B=ScalarParameter() {defaultValue=0.95, group=Snow, parameterName=Snow Roughness High}
   Alpha=@n13.R
  )
  n19 = TextureSample(UVs=Multiply(A=@n12) {constB=0.15}) {
   samplerSource=SSM_Wrap_WorldGroupSettings
   samplerType=SAMPLERTYPE_Normal
   texture=/Game/ContentCreationTools/TextureCreation/Textures/T_BPCreated_MacroNormal_01
  }
  n20 = LinearInterpolate(A=@n19.RGB, B=Constant3Vector() {constant=<rgba 0 0 1 0>}, Alpha=Constant() {r=0.7}) {constAlpha=0.8}
  FunctionOutput(None=MakeMaterialAttributes!(
    BaseColor=QualitySwitch(Default=@n17, Low=@n16, High=@n17, Medium=@n16)
    Metallic
    Specular=LinearInterpolate(Alpha=@n13.R) {constA=0.5, constB=0.1}
    Roughness=Clamp(None=@n18)
    Anisotropy
    EmissiveColor
    Opacity
    OpacityMask
    Normal=FunctionOutput(None=@n20) {outputName=Normal}
    Tangent
    WorldPositionOffset
    SubsurfaceColor
    ClearCoat
    ClearCoatRoughness
    AmbientOcclusion
    Refraction
    CustomizedUVs_0
    CustomizedUVs_1
    CustomizedUVs_2
    CustomizedUVs_3
    CustomizedUVs_4
    CustomizedUVs_5
    CustomizedUVs_6
    CustomizedUVs_7
    PixelDepthOffset
    ShadingModel
    Displacement
   )) {outputName=Material}
  n21 = Constant() {r=90000}
  n22 = LinearInterpolate(
   A=@n13.A
   B=Constant() {r=0.5}
   Alpha=Saturate(None=MaterialFunctionCall!(Distance=@n21) {materialFunction=/Game/Athena/Environments/Landscape/MaterialFunctions/MF_TerrainDistanceFade}.Result)
  )
  FunctionOutput(None=FunctionOutput(None=@n22) {outputName=Height}) {outputName=Snow_Mask_Height}
  FunctionOutput(None=@n3) {outputName=NormalDotCamera}
  Multiply(A=@n13.G, B=@n22)
 ]}
